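# Niagara system for the SunbeamQuest pickaxe's animation trails, drawn as ribbons from bones on the User skeletal mesh.
# One emitter spawns three trails as spawn groups, selecting width and vertical offset per group from arrays: a wide center and two narrower sides.
# Over each short particle lifetime, curves shift color from white to dark red, fade alpha, taper ribbon width and drive a dynamic material parameter.
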NiagaraSystem /BRCosmetics/Weapons/FORT_Melee/Pickaxe_SunbeamQuest/FX/NS_Pickaxe_SunBeamQuest_AnimTrail
dependencies {
 usedDataInterfaces [
  &NiagaraDataInterfaceArrayFloat3
  &NiagaraDataInterfaceArrayFloat
  &NiagaraDataInterfaceColorCurve
  &NiagaraDataInterfaceCurve
 ]
 usedDynamicInputs []
 usedModules [&SystemState]
}
emitters {Blades_Trails002 {
  emitterUpdateScript [
   &EmitterState()
   &FN_SpawnAnimTrail(Number of Anim Trails=3)
  ]
  particleSpawnScript [
   &FN_SetSpawnGroup(Number of Spawn Groups=3)
   &FN_InitializeAnimTrailBoneIndecies()
   &FN_InitializeAnimTrailBoneIndecies(AnimTrail Selection=1) {module=FN_InitializeAnimTrailBoneIndecies001}
   &FN_InitializeAnimTrailBoneIndecies(AnimTrail Selection=2) {module=FN_InitializeAnimTrailBoneIndecies002}
   &FN_InitializeAnimTrail(
    AnimTrailWidth=&SelectFloatFromArray(Direct Array Index=$Particles.SpawnGroup, Float Selection Array=DataInterface() {propertyValues {FloatData [0.75,0.35,0.35]}})
    Lifetime=0.25
    Position Offset=&SelectVectorFromArray(Direct Array Index=$Particles.SpawnGroup, Vector Selection Array=DataInterface() {propertyValues {FloatData [<0 0 0>, <0 0 -4>, <0 0 4>]}})
   )
   /SetVariables(Particles.PositionOffset1:Vector3f=&Add_Vector(A=Linked() {linkedVariable {name=Particles.Position, type=NiagaraPosition}}, B=<0 0 -4>)) {isSetParameters=true}
   /SetVariables(Particles.PositionOffset2:Vector3f=&Add_Vector(A=Linked() {linkedVariable {name=Particles.Position, type=NiagaraPosition}}, B=<0 0 4>)) {isSetParameters=true}
   &AddVelocity()
  ]
  particleUpdateScript [
   &FN_UpdateAnimTrail(SkeletalMesh=$User.SkeletalMesh)
   &ParticleState()
   &SolveForcesAndVelocity(Clamp Velocity=false, Limit Acceleration=false)
   &Color(Color=&ColorFromCurve(ColorCurve=DataInterface() {propertyValues {
       AlphaCurve {keys [<at 0 1 RCIM_Cubic>,<at 0.5086 0.5>,<at 1 0 RCIM_Cubic>]}
       BlueCurve {keys [<at 0 1>,<at 0.1104 1>,<at 0.2616 0.0259>]}
       GreenCurve {keys [<at 0 1>,<at 0.1104 1>,<at 0.2616 0>]}
       RedCurve {keys [<at 0 1>,<at 0.1104 1>,<at 0.2616 0.4323>]}
      }}))
   &DynamicMaterialParameters(
    Index 0 Param 1=1
    Index 0 Param 4=&FloatFromCurve(FloatCurve=DataInterface() {propertyValues {Curve {keys [<at 0 1>,<at 1 0>]}}})
    Module.Param0WriteEnabled=true
    Module.Param0WriteEnabled1=true
    Module.Param1WriteEnabled=true
    Module.Param1WriteEnabled1=true
    Module.Param2WriteEnabled=true
    Module.Param2WriteEnabled1=true
    Module.Param3WriteEnabled=true
    Module.Param3WriteEnabled1=true
    Write Parameter Index 1=true
   )
   &ScaleRibbonWidth(Ribbon Width Scale=&FloatFromCurve(FloatCurve=DataInterface() {propertyValues {Curve {keys [<at 0.05 1 RCIM_Cubic>,<at 1 0 RCIM_Cubic>]}}}))
  ]
  rendererClasses [&NiagaraRibbonRendererProperties]
  renderers [{rendererClass=&NiagaraRibbonRendererProperties}]
 }}
summary {
 systemName=NS_Pickaxe_SunBeamQuest_AnimTrail
 emitters [{emitterName=Blades_Trails002, rendererClasses [&NiagaraRibbonRendererProperties]}]
 userVariables [
  User.AnimTrailWidth {type=NiagaraFloat, defaultValue=1, description=""}
  User.ColorSwitcher {type=NiagaraFloat, defaultValue=0, description=""}
  User.Lifetime {type=NiagaraFloat, defaultValue=0.13, description=""}
  User.LoopDuration {type=NiagaraFloat, defaultValue=0.2, description=""}
  User.Ribbon camera face {type=Vector3f, description="", defaultValue {struct=Vector3f, value=<0 0 0>}}
  User.RibbonIDOffset {type=NiagaraInt32}
  User.SkeletalMesh {
   type=NiagaraDataInterfaceSkeletalMesh
   defaultValue {
    struct=NiagaraExt_VariableValue_DataInterface
    value {dataInterface=&self:NiagaraDataInterfaceSkeletalMesh_0}
   }
  }
 ]
}
