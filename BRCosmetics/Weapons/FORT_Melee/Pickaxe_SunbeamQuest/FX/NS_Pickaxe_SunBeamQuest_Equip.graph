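# Niagara equip effect for the SunBeamQuest pickaxe, combining blade trails, glints and sheen meshes.
# A ribbon emitter builds three animation trails from the user skeletal mesh, shrinking and fading from white through pink.
# Three sprite emitters follow offset bone/socket positions on that mesh as spinning lavender glints.
# Three staggered mesh bursts animate a brightening material sheen, with Y scale mirrored according to the bIsOffhand user flag.
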NiagaraSystem /BRCosmetics/Weapons/FORT_Melee/Pickaxe_SunbeamQuest/FX/NS_Pickaxe_SunBeamQuest_Equip
dependencies {
 usedDataInterfaces [
  &NiagaraDataInterfaceCurve
  &NiagaraDataInterfaceArrayFloat3
  &NiagaraDataInterfaceArrayFloat
  &NiagaraDataInterfaceColorCurve
 ]
 usedDynamicInputs []
 usedModules [&SystemState]
}
emitters {
 Blades_Trails002 {
  emitterUpdateScript [
   &EmitterState()
   &FN_SpawnAnimTrail(Number of Anim Trails=3)
  ]
  particleSpawnScript [
   &FN_SetSpawnGroup(Number of Spawn Groups=3)
   &FN_InitializeAnimTrailBoneIndecies()
   &FN_InitializeAnimTrailBoneIndecies(AnimTrail Selection=1) {module=FN_InitializeAnimTrailBoneIndecies001}
   &FN_InitializeAnimTrailBoneIndecies(AnimTrail Selection=2) {module=FN_InitializeAnimTrailBoneIndecies002}
   &FN_InitializeAnimTrail(
    AnimTrailWidth=&SelectFloatFromArray(Direct Array Index=$Particles.SpawnGroup, Float Selection Array=DataInterface() {propertyValues {FloatData [1,0.5,0.5]}})
    Lifetime=0.175
    Position Offset=&SelectVectorFromArray(Direct Array Index=$Particles.SpawnGroup, Vector Selection Array=DataInterface() {propertyValues {FloatData [<0 0 0>, <0 0 -4>, <0 0 4>]}})
   )
   /SetVariables(Particles.PositionOffset1:Vector3f=&Add_Vector(A=Linked() {linkedVariable {name=Particles.Position, type=NiagaraPosition}}, B=<0 0 -4>)) {isSetParameters=true}
   /SetVariables(Particles.PositionOffset2:Vector3f=&Add_Vector(A=Linked() {linkedVariable {name=Particles.Position, type=NiagaraPosition}}, B=<0 0 4>)) {isSetParameters=true}
   &AddVelocity()
  ]
  particleUpdateScript [
   &FN_UpdateAnimTrail(SkeletalMesh=$User.SkeletalMesh)
   &ParticleState()
   &SolveForcesAndVelocity(Clamp Velocity=false, Limit Acceleration=false)
   &Color(Color=&ColorFromCurve(ColorCurve=DataInterface() {propertyValues {
       AlphaCurve {keys [<at 0 1 RCIM_Cubic>,<at 0.5086 0.5>,<at 1 0 RCIM_Cubic>]}
       BlueCurve {keys [<at 0.1032 1>,<at 0.234 0.4094>,<at 0.4142 0.1017>]}
       GreenCurve {keys [<at 0.1032 1>,<at 0.234 0.0521>,<at 0.4142 0.0315>]}
       RedCurve {keys [<at 0.1032 1>,<at 0.234 1>,<at 0.4142 0.151>]}
      }}))
   &DynamicMaterialParameters(
    Index 0 Param 1=1
    Index 0 Param 4=&FloatFromCurve(FloatCurve=DataInterface() {propertyValues {Curve {keys [<at 0 1>,<at 1 0>]}}})
    Module.Param0WriteEnabled=true
    Module.Param0WriteEnabled1=true
    Module.Param1WriteEnabled=true
    Module.Param1WriteEnabled1=true
    Module.Param2WriteEnabled=true
    Module.Param2WriteEnabled1=true
    Module.Param3WriteEnabled=true
    Module.Param3WriteEnabled1=true
    Write Parameter Index 1=true
   )
   &ScaleRibbonWidth(Ribbon Width Scale=&FloatFromCurve(FloatCurve=DataInterface() {propertyValues {Curve {keys [<at 0.05 1 RCIM_Cubic>,<at 1 0 RCIM_Cubic>]}}}))
  ]
  rendererClasses [&NiagaraRibbonRendererProperties]
  renderers [{rendererClass=&NiagaraRibbonRendererProperties}]
 }
 Glints001 {
  emitterUpdateScript [&EmitterState(Life Cycle Mode=::NewEnumerator1, Loop Duration=2, UseLoopDelay=false)]
  particleSpawnScript [&InitializeParticle(
    Color=<rgba 0.6922 0.5444 1 0.25>
    Lifetime=&RandomRangeFloat(Minimum=0.125, Maximum=0.5)
    Sprite Rotation Mode=::NewEnumerator2
    Sprite Size Mode=::NewEnumerator3
    UsePositionOffset=false
    Write Material Random=true
   )]
  particleUpdateScript [
   &ParticleState()
   &SkeletalMeshLocation(
    Bone / Socket Index=1
    MaskSpawnGroup=false
    Mesh Sampling Type=::NewEnumerator5
    Sampled Position Offset=<2 0 4>
    Sampled Position Offset Coordinate Space:ENiagaraSkelMeshTransforms=::NewEnumerator2
    Skeletal Mesh=$User.SkeletalMesh
    Socket Sampling Mode=::NewEnumerator5
   )
   &ScaleSpriteSize(
    Uniform Curve Scale=3
    Uniform Curve Sprite Scale=DataInterface() {propertyValues {Curve {keys [
        {arriveTangent=9.0725, interpMode=RCIM_Cubic, leaveTangent=9.0725, tangentMode=RCTM_User, time=0, value=0}
        {arriveTangent=-1.7917, interpMode=RCIM_Cubic, leaveTangent=-1.7917, tangentMode=RCTM_User, time=0.0666, value=0.9847}
        {arriveTangent=-0.0405, interpMode=RCIM_Cubic, leaveTangent=-0.0405, tangentMode=RCTM_User, time=1, value=0}
       ]}}}
   )
   &SpriteRotationRate(Rotation Rate=&RandomRangeFloat(Minimum=150, Maximum=180))
   &CameraOffset(Camera Offset Amount=2)
  ]
  rendererClasses [&NiagaraSpriteRendererProperties]
  renderers [{rendererClass=&NiagaraSpriteRendererProperties}]
 }
 Glints002 {
  emitterUpdateScript [&EmitterState(Life Cycle Mode=::NewEnumerator1, Loop Duration=2, UseLoopDelay=false)]
  particleSpawnScript [&InitializeParticle(
    Color=<rgba 0.6922 0.5444 1 0.25>
    Lifetime=&RandomRangeFloat(Minimum=0.125, Maximum=0.5)
    Sprite Rotation Mode=::NewEnumerator2
    Sprite Size Mode=::NewEnumerator3
    UsePositionOffset=false
    Write Material Random=true
   )]
  particleUpdateScript [
   &ParticleState()
   &SkeletalMeshLocation(
    Bone / Socket Index=1
    MaskSpawnGroup=false
    Mesh Sampling Type=::NewEnumerator5
    Sampled Position Offset Coordinate Space:ENiagaraSkelMeshTransforms=::NewEnumerator2
    Skeletal Mesh=$User.SkeletalMesh
    Socket Sampling Mode=::NewEnumerator5
   )
   &ScaleSpriteSize(
    Uniform Curve Scale=3
    Uniform Curve Sprite Scale=DataInterface() {propertyValues {Curve {keys [
        {arriveTangent=9.0725, interpMode=RCIM_Cubic, leaveTangent=9.0725, tangentMode=RCTM_User, time=0, value=0}
        {arriveTangent=-1.7917, interpMode=RCIM_Cubic, leaveTangent=-1.7917, tangentMode=RCTM_User, time=0.0666, value=0.9847}
        {arriveTangent=-0.0405, interpMode=RCIM_Cubic, leaveTangent=-0.0405, tangentMode=RCTM_User, time=1, value=0}
       ]}}}
   )
   &SpriteRotationRate(Rotation Rate=&RandomRangeFloat(Minimum=150, Maximum=180))
   &CameraOffset(Camera Offset Amount=2)
  ]
  rendererClasses [&NiagaraSpriteRendererProperties]
  renderers [{rendererClass=&NiagaraSpriteRendererProperties}]
 }
 Glints003 {
  emitterUpdateScript [&EmitterState(Life Cycle Mode=::NewEnumerator1, Loop Duration=2, UseLoopDelay=false)]
  particleSpawnScript [&InitializeParticle(
    Color=<rgba 0.6922 0.5444 1 0.25>
    Lifetime=&RandomRangeFloat(Minimum=0.125, Maximum=0.5)
    Sprite Rotation Mode=::NewEnumerator2
    Sprite Size Mode=::NewEnumerator3
    UsePositionOffset=false
    Write Material Random=true
   )]
  particleUpdateScript [
   &ParticleState()
   &SkeletalMeshLocation(
    Bone / Socket Index=1
    MaskSpawnGroup=false
    Mesh Sampling Type=::NewEnumerator5
    Sampled Position Offset=<0.75 0 -4.5>
    Sampled Position Offset Coordinate Space:ENiagaraSkelMeshTransforms=::NewEnumerator2
    Skeletal Mesh=$User.SkeletalMesh
    Socket Sampling Mode=::NewEnumerator5
   )
   &ScaleSpriteSize(
    Uniform Curve Scale=3
    Uniform Curve Sprite Scale=DataInterface() {propertyValues {Curve {keys [
        {arriveTangent=9.0725, interpMode=RCIM_Cubic, leaveTangent=9.0725, tangentMode=RCTM_User, time=0, value=0}
        {arriveTangent=-1.7917, interpMode=RCIM_Cubic, leaveTangent=-1.7917, tangentMode=RCTM_User, time=0.0666, value=0.9847}
        {arriveTangent=-0.0405, interpMode=RCIM_Cubic, leaveTangent=-0.0405, tangentMode=RCTM_User, time=1, value=0}
       ]}}}
   )
   &SpriteRotationRate(Rotation Rate=&RandomRangeFloat(Minimum=150, Maximum=180))
   &CameraOffset(Camera Offset Amount=2)
  ]
  rendererClasses [&NiagaraSpriteRendererProperties]
  renderers [{rendererClass=&NiagaraSpriteRendererProperties}]
 }
 SheenMesh001 {
  emitterUpdateScript [
   &EmitterState()
   &SpawnBurst_Instantaneous(Use Spawn Probability=false, UseLoopCountLimit=false)
  ]
  particleSpawnScript [
   &ApplyInitialForces()
   &InitializeParticle(
    Color=<rgba 2.0781 1.1562 3 0.2>
    Lifetime=0.3
    Mesh Scale=&MakeCustomVectorFromBool(Bool=$User.bIsOffhand, True Vector=<1 1 1>, False Vector=<1 -1 1>)
    Mesh Scale Mode=::NewEnumerator4
    Position Offset=&MakeCustomVectorFromBool(Bool=$User.bIsOffhand, True Vector=<0 0 0>, False Vector=<0 0 0>)
    UsePositionOffset=true
    Write Material Random=true
   )
  ]
  particleUpdateScript [
   &SolveForcesAndVelocity(Clamp Velocity=false, Limit Acceleration=false)
   &ParticleState()
   &GenerateDeathEvent(Use Event Probability=false)
   &DynamicMaterialParameters(
    Index 0 Param 1=&Lerp_Float(
     A=0
     B=0.7182
     Alpha=&FloatFromCurve(FloatCurve=DataInterface() {propertyValues {Curve {keys [
          {arriveTangent=1.8514, interpMode=RCIM_Cubic, leaveTangent=1.8514, tangentMode=RCTM_User, time=0, value=0}
          {arriveTangent=1.3047, interpMode=RCIM_Cubic, leaveTangent=1.3047, tangentMode=RCTM_User, time=1, value=1}
         ]}}})
    )
    Index 0 Param 2=0.75
    Module.Param0WriteEnabled=true
    Module.Param1WriteEnabled=true
    Module.Param2WriteEnabled=true
    Module.Param3WriteEnabled=true
   )
   &ScaleColor(
    Scale RGB=&VectorFromFloat(Value=&FloatFromCurve(FloatCurve=DataInterface() {propertyValues {Curve {keys [<at 0 0 RCIM_Cubic>, {arriveTangent=3, interpMode=RCIM_Cubic, leaveTangent=3, time=0.9, value=1.5}, <at 1 3 RCIM_Cubic>]}}}))
    ScaleA=true
    ScaleRGB=true
   )
  ]
  rendererClasses [&NiagaraMeshRendererProperties]
  renderers [{rendererClass=&NiagaraMeshRendererProperties, rendererIndex=0}]
 }
 SheenMesh002 {
  emitterUpdateScript [
   &EmitterState()
   &SpawnBurst_Instantaneous(Spawn Time=0.075, Use Spawn Probability=false, UseLoopCountLimit=false)
  ]
  particleSpawnScript [
   &ApplyInitialForces()
   &InitializeParticle(
    Color=<rgba 2.0781 1.1562 3 0.2>
    Lifetime=0.3
    Mesh Scale=&MakeCustomVectorFromBool(Bool=$User.bIsOffhand, True Vector=<1 1 1>, False Vector=<1 -1 1>)
    Mesh Scale Mode=::NewEnumerator4
    Position Offset=&MakeCustomVectorFromBool(Bool=$User.bIsOffhand, True Vector=<0 0 0>, False Vector=<0 0 0>)
    UsePositionOffset=true
    Write Material Random=true
   )
  ]
  particleUpdateScript [
   &SolveForcesAndVelocity(Clamp Velocity=false, Limit Acceleration=false)
   &ParticleState()
   &GenerateDeathEvent(Use Event Probability=false)
   &DynamicMaterialParameters(
    Index 0 Param 1=&Lerp_Float(
     A=0
     B=0.7182
     Alpha=&FloatFromCurve(FloatCurve=DataInterface() {propertyValues {Curve {keys [
          {arriveTangent=1.8514, interpMode=RCIM_Cubic, leaveTangent=1.8514, tangentMode=RCTM_User, time=0, value=0}
          {arriveTangent=1.3047, interpMode=RCIM_Cubic, leaveTangent=1.3047, tangentMode=RCTM_User, time=1, value=1}
         ]}}})
    )
    Index 0 Param 2=0.75
    Module.Param0WriteEnabled=true
    Module.Param1WriteEnabled=true
    Module.Param2WriteEnabled=true
    Module.Param3WriteEnabled=true
   )
   &ScaleColor(
    Scale RGB=&VectorFromFloat(Value=&FloatFromCurve(FloatCurve=DataInterface() {propertyValues {Curve {keys [<at 0 0 RCIM_Cubic>, {arriveTangent=3, interpMode=RCIM_Cubic, leaveTangent=3, time=0.9, value=1.5}, <at 1 3 RCIM_Cubic>]}}}))
    ScaleA=true
    ScaleRGB=true
   )
  ]
  rendererClasses [&NiagaraMeshRendererProperties]
  renderers [{rendererClass=&NiagaraMeshRendererProperties, rendererIndex=0}]
 }
 SheenMesh003 {
  emitterUpdateScript [
   &EmitterState()
   &SpawnBurst_Instantaneous(Spawn Time=0.12, Use Spawn Probability=false, UseLoopCountLimit=false)
  ]
  particleSpawnScript [
   &ApplyInitialForces()
   &InitializeParticle(
    Color=<rgba 2.0781 1.1562 3 0.2>
    Lifetime=0.3
    Mesh Scale=&MakeCustomVectorFromBool(Bool=$User.bIsOffhand, True Vector=<1 1 1>, False Vector=<1 -1 1>)
    Mesh Scale Mode=::NewEnumerator4
    Position Offset=&MakeCustomVectorFromBool(Bool=$User.bIsOffhand, True Vector=<0 0 0>, False Vector=<0 0 0>)
    UsePositionOffset=true
    Write Material Random=true
   )
  ]
  particleUpdateScript [
   &SolveForcesAndVelocity(Clamp Velocity=false, Limit Acceleration=false)
   &ParticleState()
   &GenerateDeathEvent(Use Event Probability=false)
   &DynamicMaterialParameters(
    Index 0 Param 1=&Lerp_Float(
     A=0
     B=0.7182
     Alpha=&FloatFromCurve(FloatCurve=DataInterface() {propertyValues {Curve {keys [
          {arriveTangent=1.8514, interpMode=RCIM_Cubic, leaveTangent=1.8514, tangentMode=RCTM_User, time=0, value=0}
          {arriveTangent=1.3047, interpMode=RCIM_Cubic, leaveTangent=1.3047, tangentMode=RCTM_User, time=1, value=1}
         ]}}})
    )
    Index 0 Param 2=0.75
    Module.Param0WriteEnabled=true
    Module.Param1WriteEnabled=true
    Module.Param2WriteEnabled=true
    Module.Param3WriteEnabled=true
   )
   &ScaleColor(
    Scale RGB=&VectorFromFloat(Value=&FloatFromCurve(FloatCurve=DataInterface() {propertyValues {Curve {keys [<at 0 0 RCIM_Cubic>, {arriveTangent=3, interpMode=RCIM_Cubic, leaveTangent=3, time=0.9, value=1.5}, <at 1 3 RCIM_Cubic>]}}}))
    ScaleA=true
    ScaleRGB=true
   )
  ]
  rendererClasses [&NiagaraMeshRendererProperties]
  renderers [{rendererClass=&NiagaraMeshRendererProperties, rendererIndex=0}]
 }
}
summary {
 systemName=NS_Pickaxe_SunBeamQuest_Equip
 emitters [
  {emitterName=Glints002, rendererClasses [&NiagaraSpriteRendererProperties]}
  {emitterName=SheenMesh002, rendererClasses [&NiagaraMeshRendererProperties]}
  {emitterName=SheenMesh001, rendererClasses [&NiagaraMeshRendererProperties]}
  {emitterName=SheenMesh003, rendererClasses [&NiagaraMeshRendererProperties]}
  {emitterName=Glints001, rendererClasses [&NiagaraSpriteRendererProperties]}
  {emitterName=Glints003, rendererClasses [&NiagaraSpriteRendererProperties]}
  {emitterName=Blades_Trails002, rendererClasses [&NiagaraRibbonRendererProperties]}
 ]
 userVariables [
  User.AnimTrailWidth {type=NiagaraFloat, defaultValue=1.7, description=""}
  User.bIsOffhand {type=NiagaraBool}
  User.Delay {type=NiagaraFloat, defaultValue=0.1, description=""}
  User.Lifetime {type=NiagaraFloat, defaultValue=0.15, description=""}
  User.LoopDuration {type=NiagaraFloat, defaultValue=0.07, description=""}
  User.RibbonIDOffset {type=NiagaraInt32}
  User.SkeletalMesh {
   type=NiagaraDataInterfaceSkeletalMesh
   defaultValue {
    struct=NiagaraExt_VariableValue_DataInterface
    value {dataInterface=&self:NiagaraDataInterfaceSkeletalMesh_0}
   }
  }
 ]
}
